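# Physics asset for the medium female DyedDuelist player body, covering the dynamic chest ribbons and the upper spine around them.
# Thin box bodies form a two-link left and a three-link right ribbon chain, both hung from a small spine_05 capsule.
# Two spine_04 capsules give torso volume.
# Ribbon joints share one set of angular limits, with a wider first swing than second swing or twist, while the spine_05-to-spine_04 joint rotates freely.
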PhysicsAsset /Game/Characters/Player/Female/Medium/Bodies/F_MED_DyedDuelist/Meshes/F_MED_DyedDuelist_Physics
bodies {
 dyn_L_chest_ribbon_1 [{shapeType=Box, extentX=6.9098, extentY=0.7, extentZ=2.982, shapeName=dyn_L_chest_ribbon_1_box, center=<1 0 0>}]
 dyn_L_chest_ribbon_2 [{shapeType=Box, extentX=8.1412, extentY=1, extentZ=2.982, shapeName=dyn_L_chest_ribbon_2_box, center=<3.5 -0.1105 0>}]
 dyn_R_chest_ribbon_1 [{shapeType=Box, extentX=6.9098, extentY=0.7, extentZ=2.982, shapeName=dyn_R_chest_ribbon_1_box, center=<1 0 0>}]
 dyn_R_chest_ribbon_2 [{shapeType=Box, extentX=6.9098, extentY=0.7, extentZ=2.982, shapeName=dyn_R_chest_ribbon_2_box, center=<1 0.228 0>}]
 dyn_R_chest_ribbon_3 [{shapeType=Box, extentX=6.9098, extentY=0.7, extentZ=2.982, shapeName=dyn_R_chest_ribbon_3_box, center=<1 0 0>}]
 spine_04 [
  {shapeType=Capsule, length=5, radius=11.2425, shapeName=spine_04_capsule, center=<-0.9464 0.3707 0>}
  {shapeType=Capsule, length=3, radius=1.7, shapeName=spine_04_capsule, center=<0.7663 11.7784 0>, rotation=<rot 90 0 0>}
 ]
 spine_05 [{shapeType=Capsule, length=1, radius=1, shapeName=spine_05_capsule}]
}
constraints [
 {
  bone1Name=spine_05
  bone2Name=spine_04
  swing1LimitDegrees=45
  swing1Motion=Free
  swing2LimitDegrees=45
  swing2Motion=Free
  twistLimitDegrees=45
  twistMotion=Free
 }
 {bone1Name=dyn_L_chest_ribbon_1, bone2Name=spine_05, swing1LimitDegrees=45, swing2LimitDegrees=10, twistLimitDegrees=15}
 {bone1Name=dyn_L_chest_ribbon_2, bone2Name=dyn_L_chest_ribbon_1, swing1LimitDegrees=45, swing2LimitDegrees=10, twistLimitDegrees=15}
 {bone1Name=dyn_R_chest_ribbon_1, bone2Name=spine_05, swing1LimitDegrees=45, swing2LimitDegrees=10, twistLimitDegrees=15}
 {bone1Name=dyn_R_chest_ribbon_2, bone2Name=dyn_R_chest_ribbon_1, swing1LimitDegrees=45, swing2LimitDegrees=10, twistLimitDegrees=15}
 {bone1Name=dyn_R_chest_ribbon_3, bone2Name=dyn_R_chest_ribbon_2, swing1LimitDegrees=45, swing2LimitDegrees=10, twistLimitDegrees=15}
]
