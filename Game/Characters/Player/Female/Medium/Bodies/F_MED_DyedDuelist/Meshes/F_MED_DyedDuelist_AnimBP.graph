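# Animation Blueprint for the Dyed Duelist female medium player body, copying the attached parent mesh's pose and adding secondary motion to costume bones.
# One long chain first constrains cape and backpack attach bones to the torso, then runs a RigidBody simulation.
# Bone constraints, spring-based AnimDynamics and additive bone offsets then layer motion onto frills, ribbons and hip ruffles.
# Crouch state variables switch right front ribbon constraint weighting to limit how it follows the thigh.
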AnimBlueprint /Game/Characters/Player/Female/Medium/Bodies/F_MED_DyedDuelist/Meshes/F_MED_DyedDuelist_AnimBP
target_skeleton=&/F_MED_DyedDuelist_Skeleton
graphs [
 AnimGraph {nodes [
   n1 = "Constraint - attach_cape (P)"(ComponentPose=Local To Component(LocalPose=Copy Pose From Mesh() {bUseAttachedParent=true}), Alpha=1, ConstraintWeights_0=1) {
    alphaBoolBlend {blendOption=Linear}
    boneToModify {boneName=attach_cape}
    constraintSetup [{targetBone {boneName=spine_05}}]
    constraintWeights [1]
   }
   n2 = "Constraint - attach_backpack (PPPPP)"!(
    ComponentPose=@n1
    Alpha=1
    ConstraintWeights_0=0.0161
    ConstraintWeights_1:float=0.0162
    ConstraintWeights_2:float=0.2373
    ConstraintWeights_3:float=0.6981
    ConstraintWeights_4:float=0.0323
   ) -> (Pose) {
    alphaBoolBlend {blendOption=Linear}
    boneToModify {boneName=attach_backpack}
    constraintSetup [
     {targetBone {boneName=clavicle_r}}
     {targetBone {boneName=clavicle_l}}
     {targetBone {boneName=spine_04}}
     {targetBone {boneName=spine_05}}
     {targetBone {boneName=spine_03}}
    ]
    constraintWeights [1,1,1,1,1]
   }
   n3 = RigidBody!(ComponentPose=@n2, Alpha, OverrideWorldGravity:Vector="0,0,-500", ExternalForce) -> (Pose) {
    bClampLinearTranslationLimitToRefPose=true
    bOverrideWorldGravity=true
    overridePhysicsAsset=&/F_MED_DyedDuelist_Physics
    simulationSpace=BaseBoneSpace
    alphaBoolBlend {blendOption=Linear}
    baseBoneRef {boneName=spine_03}
    componentLinearAccScale=<0.3 0.3 0.35>
    componentLinearVelScale=<0.32 0.32 0.37>
    overrideWorldGravity {z=-980}
    simSpaceSettings {maxAngularAcceleration=300, maxAngularVelocity=5, maxLinearAcceleration=300, maxLinearVelocity=400, externalLinearDragV=<0.7 0.7 0.7>}
   }
   n4 = Anim Dynamics - Bone: dyn_L_upper_wrist_frill_back!(SimulationSpace:AnimPhysSimSpaceType Enum, ComponentPose=@n3, Alpha=1) -> (Pose) {
    angularDampingOverride=0.9999
    angularSpringConstant=2000
    bAngularSpring=true
    bOverrideAngularDamping=true
    bUseGravityOverride=true
    gravityScale=0.35
    simulationSpace=RootRelative
    alphaBoolBlend {blendOption=Linear}
    alphaScaleBiasClamp {scale=0.85}
    boundBone {boneName=dyn_L_upper_wrist_frill_back}
    componentLinearAccScale=<1 1 1>
    componentLinearVelScale=<1 1 1>
    gravityOverride {z=-980}
    physicsBodyDefinitions [{
      boundBone {boneName=dyn_L_upper_wrist_frill_back}
      boxExtents=<10 10 10>
      constraintSetup {angularLimitsMax=<0 0 -2>, angularLimitsMin=<0 -20 -20>, angularTarget=<1 0 0>}
      localJointOffset=<1 0 0>
     }]
    relativeSpaceBone {boneName=upperarm_l}
   }
   n5 = Anim Dynamics - Bone: dyn_R_upper_wrist_frill_back!(SimulationSpace:AnimPhysSimSpaceType Enum, ComponentPose=@n4, Alpha=1) -> (Pose) {
    angularDampingOverride=0.9999
    angularSpringConstant=2000
    bAngularSpring=true
    bOverrideAngularDamping=true
    bUseGravityOverride=true
    gravityScale=0.35
    simulationSpace=RootRelative
    alphaBoolBlend {blendOption=Linear}
    alphaScaleBiasClamp {scale=0.85}
    boundBone {boneName=dyn_R_upper_wrist_frill_back}
    componentLinearAccScale=<1 1 1>
    componentLinearVelScale=<1 1 1>
    gravityOverride {z=-980}
    physicsBodyDefinitions [{
      boundBone {boneName=dyn_R_upper_wrist_frill_back}
      boxExtents=<10 10 10>
      constraintSetup {angularLimitsMax=<0 0 0>, angularLimitsMin=<0 -20 -20>, angularTarget=<1 0 0>}
      localJointOffset=<-1 0 0>
     }]
    relativeSpaceBone {boneName=upperarm_r}
   }
   n6 = Anim Dynamics - Bone: dyn_L_upper_wrist_frill_front!(SimulationSpace:AnimPhysSimSpaceType Enum, ComponentPose=@n5, Alpha=1) -> (Pose) {
    angularDampingOverride=0.9999
    angularSpringConstant=2000
    bAngularSpring=true
    bOverrideAngularDamping=true
    bUseGravityOverride=true
    gravityScale=0.35
    simulationSpace=RootRelative
    alphaBoolBlend {blendOption=Linear}
    alphaScaleBiasClamp {scale=0.85}
    boundBone {boneName=dyn_L_upper_wrist_frill_front}
    componentLinearAccScale=<1 1 1>
    componentLinearVelScale=<1 1 1>
    gravityOverride {z=-980}
    physicsBodyDefinitions [{
      boundBone {boneName=dyn_L_upper_wrist_frill_front}
      boxExtents=<10 10 10>
      constraintSetup {angularLimitsMax=<0 0 20>, angularLimitsMin=<0 -20 2>, angularTarget=<1 0 0>}
      localJointOffset=<1 0 0>
     }]
    relativeSpaceBone {boneName=upperarm_l}
   }
   n7 = Anim Dynamics - Bone: dyn_R_upper_wrist_frill_front!(SimulationSpace:AnimPhysSimSpaceType Enum, ComponentPose=@n6, Alpha=1) -> (Pose) {
    angularDampingOverride=0.9999
    angularSpringConstant=2000
    bAngularSpring=true
    bOverrideAngularDamping=true
    bUseGravityOverride=true
    gravityScale=0.35
    simulationSpace=RootRelative
    alphaBoolBlend {blendOption=Linear}
    alphaScaleBiasClamp {scale=0.85}
    boundBone {boneName=dyn_R_upper_wrist_frill_front}
    componentLinearAccScale=<1 1 1>
    componentLinearVelScale=<1 1 1>
    gravityOverride {z=-980}
    physicsBodyDefinitions [{
      boundBone {boneName=dyn_R_upper_wrist_frill_front}
      boxExtents=<10 10 10>
      constraintSetup {angularLimitsMax=<0 0 20>, angularLimitsMin=<0 -20 0>, angularTarget=<1 0 0>}
      localJointOffset=<-1 0 0>
     }]
    relativeSpaceBone {boneName=upperarm_r}
   }
   n8 = Anim Dynamics - Bone: dyn_L_lower_wrist_frill_back!(SimulationSpace:AnimPhysSimSpaceType Enum, ComponentPose=@n7, Alpha=1) -> (Pose) {
    angularDampingOverride=0.9999
    angularSpringConstant=2000
    bAngularSpring=true
    bOverrideAngularDamping=true
    bUseGravityOverride=true
    gravityScale=0.35
    simulationSpace=RootRelative
    alphaBoolBlend {blendOption=Linear}
    alphaScaleBiasClamp {scale=0.85}
    boundBone {boneName=dyn_L_lower_wrist_frill_back}
    componentLinearAccScale=<1 1 1>
    componentLinearVelScale=<1 1 1>
    gravityOverride {z=-980}
    physicsBodyDefinitions [{
      boundBone {boneName=dyn_L_lower_wrist_frill_back}
      boxExtents=<10 10 10>
      constraintSetup {angularLimitsMax=<0 0 0>, angularLimitsMin=<0 -30 -45>, angularTarget=<1 0 0>}
      localJointOffset=<1 0 0>
     }]
    relativeSpaceBone {boneName=upperarm_l}
   }
   n9 = Anim Dynamics - Bone: dyn_R_lower_wrist_frill_back!(SimulationSpace:AnimPhysSimSpaceType Enum, ComponentPose=@n8, Alpha=1) -> (Pose) {
    angularDampingOverride=0.9999
    angularSpringConstant=2000
    bAngularSpring=true
    bOverrideAngularDamping=true
    bUseGravityOverride=true
    gravityScale=0.35
    simulationSpace=RootRelative
    alphaBoolBlend {blendOption=Linear}
    alphaScaleBiasClamp {scale=0.85}
    boundBone {boneName=dyn_R_lower_wrist_frill_back}
    componentLinearAccScale=<1 1 1>
    componentLinearVelScale=<1 1 1>
    gravityOverride {z=-980}
    physicsBodyDefinitions [{
      boundBone {boneName=dyn_R_lower_wrist_frill_back}
      boxExtents=<10 10 10>
      constraintSetup {angularLimitsMax=<0 0 0>, angularLimitsMin=<0 -30 -45>, angularTarget=<1 0 0>}
      localJointOffset=<-1 0 0>
     }]
    relativeSpaceBone {boneName=upperarm_r}
   }
   n10 = Anim Dynamics - Bone: dyn_L_lower_wrist_frill_front!(SimulationSpace:AnimPhysSimSpaceType Enum, ComponentPose=@n9, Alpha=1) -> (Pose) {
    angularDampingOverride=0.9999
    angularSpringConstant=2000
    bAngularSpring=true
    bOverrideAngularDamping=true
    bUseGravityOverride=true
    gravityScale=0.35
    simulationSpace=RootRelative
    alphaBoolBlend {blendOption=Linear}
    alphaScaleBiasClamp {scale=0.85}
    boundBone {boneName=dyn_L_lower_wrist_frill_front}
    componentLinearAccScale=<1 1 1>
    componentLinearVelScale=<1 1 1>
    gravityOverride {z=-980}
    physicsBodyDefinitions [{
      boundBone {boneName=dyn_L_lower_wrist_frill_front}
      boxExtents=<10 10 10>
      constraintSetup {angularLimitsMax=<0 0 45>, angularLimitsMin=<0 -30 0>, angularTarget=<1 0 0>}
      localJointOffset=<1 0 0>
     }]
    relativeSpaceBone {boneName=upperarm_l}
   }
   n11 = Anim Dynamics - Bone: dyn_R_lower_wrist_frill_front!(SimulationSpace:AnimPhysSimSpaceType Enum, ComponentPose=@n10, Alpha=1) -> (Pose) {
    angularDampingOverride=0.9999
    angularSpringConstant=2000
    bAngularSpring=true
    bOverrideAngularDamping=true
    bUseGravityOverride=true
    gravityScale=0.35
    simulationSpace=RootRelative
    alphaBoolBlend {blendOption=Linear}
    alphaScaleBiasClamp {scale=0.85}
    boundBone {boneName=dyn_R_lower_wrist_frill_front}
    componentLinearAccScale=<1 1 1>
    componentLinearVelScale=<1 1 1>
    gravityOverride {z=-980}
    physicsBodyDefinitions [{
      boundBone {boneName=dyn_R_lower_wrist_frill_front}
      boxExtents=<10 10 10>
      constraintSetup {angularLimitsMax=<0 0 45>, angularLimitsMin=<0 -30 0>, angularTarget=<1 0 0>}
      localJointOffset=<-1 0 0>
     }]
    relativeSpaceBone {boneName=upperarm_r}
   }
   n12 = "Constraint - dyn_L_frontHip_metalPart (PP)"!(ComponentPose=@n11, Alpha=1, ConstraintWeights_0=0.8, ConstraintWeights_1:float=0.2) -> (Pose) {
    alphaBoolBlend {blendOption=Linear}
    boneToModify {boneName=dyn_L_frontHip_metalPart}
    constraintSetup [{targetBone {boneName=pelvis}}, {targetBone {boneName=thigh_l}}]
    constraintWeights [1,1]
   }
   n13 = "Constraint - dyn_R_frontHip_metalPart (PP)"!(ComponentPose=@n12, Alpha=1, ConstraintWeights_0=0.8, ConstraintWeights_1:float=0.2) -> (Pose) {
    alphaBoolBlend {blendOption=Linear}
    boneToModify {boneName=dyn_R_frontHip_metalPart}
    constraintSetup [{targetBone {boneName=pelvis}}, {targetBone {boneName=thigh_r}}]
    constraintWeights [1,1]
   }
   n14 = "Constraint - dyn_R_front_ribbon_base (PP)"!(ComponentPose=@n13, Alpha=1, ConstraintWeights_0=0.35, ConstraintWeights_1:float=0.65) -> (Pose) {
    alphaBoolBlend {blendOption=Linear}
    boneToModify {boneName=dyn_R_front_ribbon_base}
    constraintSetup [{targetBone {boneName=thigh_r}}, {targetBone {boneName=pelvis}}]
    constraintWeights [1,1]
   }
   n15 = "Constraint - dyn_R_front_ribbon_A (PP)"!(ComponentPose=@n14, Alpha=1, ConstraintWeights_0=0.9, ConstraintWeights_1:float=0.1) -> (Pose) {
    alphaBoolBlend {blendOption=Linear}
    boneToModify {boneName=dyn_R_front_ribbon_A}
    constraintSetup [{targetBone {boneName=thigh_r}}, {targetBone {boneName=pelvis}}]
    constraintWeights [1,1]
   }
   n16 = Get isNotCrouching() -> (isNotCrouching:double)
   n17 = "Constraint - dyn_R_front_ribbon_B_1 (PP)"!(ComponentPose=@n15, Alpha=@n16, ConstraintWeights_0=0.75, ConstraintWeights_1:float=0.25) -> (Pose) {
    alphaBoolBlend {blendOption=Linear}
    boneToModify {boneName=dyn_R_front_ribbon_B_1}
    constraintSetup [{targetBone {boneName=thigh_r}}, {targetBone {boneName=pelvis}}]
    constraintWeights [1,1]
   }
   n18 = "Constraint - dyn_R_front_ribbon_B_2 (PP)"!(ComponentPose=@n17, Alpha=@n16, ConstraintWeights_0=1, ConstraintWeights_1:float=0) -> (Pose) {
    alphaBoolBlend {blendOption=Linear}
    boneToModify {boneName=dyn_R_front_ribbon_B_2}
    constraintSetup [{targetBone {boneName=thigh_r}}, {targetBone {boneName=pelvis}}]
    constraintWeights [1,1]
   }
   n19 = "Constraint - dyn_R_front_ribbon_B_3 (PP)"!(ComponentPose=@n18, Alpha=@n16, ConstraintWeights_0=1, ConstraintWeights_1:float=0) -> (Pose) {
    alphaBoolBlend {blendOption=Linear}
    boneToModify {boneName=dyn_R_front_ribbon_B_3}
    constraintSetup [{targetBone {boneName=thigh_r}}, {targetBone {boneName=pelvis}}]
    constraintWeights [1,1]
   }
   n20 = Anim Dynamics - Bone: dyn_R_front_ribbon_base!(SimulationSpace:AnimPhysSimSpaceType Enum, ComponentPose=@n19, Alpha=1) -> (Pose) {
    angularDampingOverride=0.9999
    angularSpringConstant=200
    bAngularSpring=true
    bOverrideAngularDamping=true
    bUseGravityOverride=true
    gravityScale=0.35
    simulationSpace=RootRelative
    alphaBoolBlend {blendOption=Linear}
    boundBone {boneName=dyn_R_front_ribbon_base}
    componentLinearAccScale=<1 1 1>
    componentLinearVelScale=<1 1 1>
    gravityOverride {z=-2000}
    physicsBodyDefinitions [{
      boundBone {boneName=dyn_R_front_ribbon_base}
      boxExtents=<20 20 20>
      constraintSetup {angularLimitsMax=<0 10 0>, angularLimitsMin=<0 -10 0>, angularTarget=<1 0 0>}
      localJointOffset=<2 0 0>
     }]
   }
   n21 = Anim Dynamics - Bone: dyn_R_front_ribbon_B_1!(SimulationSpace:AnimPhysSimSpaceType Enum, ComponentPose=@n20, Alpha=1) -> (Pose) {
    angularDampingOverride=0.9999
    angularSpringConstant=200
    bAngularSpring=true
    bOverrideAngularDamping=true
    bUseGravityOverride=true
    gravityScale=0.35
    simulationSpace=RootRelative
    alphaBoolBlend {blendOption=Linear}
    boundBone {boneName=dyn_R_front_ribbon_B_1}
    componentLinearAccScale=<1 1 1>
    componentLinearVelScale=<1 1 1>
    gravityOverride {z=-2000}
    physicsBodyDefinitions [{
      boundBone {boneName=dyn_R_front_ribbon_B_1}
      boxExtents=<20 20 20>
      constraintSetup {angularLimitsMax=<0 2 0>, angularLimitsMin=<0 -45 -50>, angularTarget=<1 0 0>}
      localJointOffset=<2 0 0>
     }]
   }
   n22 = Anim Dynamics - Bone: dyn_R_front_ribbon_B_2!(SimulationSpace:AnimPhysSimSpaceType Enum, ComponentPose=@n21, Alpha=1) -> (Pose) {
    angularDampingOverride=0.9999
    angularSpringConstant=200
    bAngularSpring=true
    bOverrideAngularDamping=true
    bUseGravityOverride=true
    gravityScale=0.35
    simulationSpace=RootRelative
    alphaBoolBlend {blendOption=Linear}
    boundBone {boneName=dyn_R_front_ribbon_B_2}
    componentLinearAccScale=<1 1 1>
    componentLinearVelScale=<1 1 1>
    gravityOverride {z=-2000}
    physicsBodyDefinitions [{
      boundBone {boneName=dyn_R_front_ribbon_B_2}
      boxExtents=<20 20 20>
      constraintSetup {angularLimitsMax=<0 2 0>, angularLimitsMin=<0 -45 -50>, angularTarget=<1 0 0>}
      localJointOffset=<2 0 0>
     }]
   }
   n23 = Anim Dynamics - Bone: dyn_R_front_ribbon_B_3!(SimulationSpace:AnimPhysSimSpaceType Enum, ComponentPose=@n22, Alpha=1) -> (Pose) {
    angularDampingOverride=0.9999
    angularSpringConstant=200
    bAngularSpring=true
    bOverrideAngularDamping=true
    bUseGravityOverride=true
    gravityScale=0.35
    simulationSpace=RootRelative
    alphaBoolBlend {blendOption=Linear}
    boundBone {boneName=dyn_R_front_ribbon_B_3}
    componentLinearAccScale=<1 1 1>
    componentLinearVelScale=<1 1 1>
    gravityOverride {z=-2000}
    physicsBodyDefinitions [{
      boundBone {boneName=dyn_R_front_ribbon_B_3}
      boxExtents=<20 20 20>
      constraintSetup {angularLimitsMax=<0 2 0>, angularLimitsMin=<0 -45 -50>, angularTarget=<1 0 0>}
      localJointOffset=<2 0 0>
     }]
   }
   n24 = Anim Dynamics - Bone: dyn_R_front_ribbon_A!(SimulationSpace:AnimPhysSimSpaceType Enum, ComponentPose=@n23, Alpha=1) -> (Pose) {
    angularDampingOverride=0.9999
    angularSpringConstant=200
    bAngularSpring=true
    bOverrideAngularDamping=true
    bUseGravityOverride=true
    gravityScale=0.35
    simulationSpace=RootRelative
    alphaBoolBlend {blendOption=Linear}
    boundBone {boneName=dyn_R_front_ribbon_A}
    componentLinearAccScale=<1 1 1>
    componentLinearVelScale=<1 1 1>
    gravityOverride {z=-2000}
    physicsBodyDefinitions [{
      boundBone {boneName=dyn_R_front_ribbon_A}
      boxExtents=<20 20 20>
      constraintSetup {angularLimitsMax=<0 40 -1>, angularLimitsMin=<0 -5 -50>, angularTarget=<1 0 0>}
      localJointOffset=<2 0 0>
     }]
   }
   n25 = Get isCrouching() -> (isCrouching:double)
   n26 = "Constraint - dyn_R_front_ribbon_B_1 (P)"(ComponentPose=@n24, Alpha=@n25, ConstraintWeights_0=1) {
    alphaBoolBlend {blendOption=Linear}
    alphaScaleBiasClamp {scale=0.6}
    boneToModify {boneName=dyn_R_front_ribbon_B_1}
    constraintSetup [{targetBone {boneName=thigh_r}}]
    constraintWeights [1]
   }
   n27 = "Constraint - dyn_R_front_ribbon_B_2 (P)"(ComponentPose=@n26, Alpha=@n25, ConstraintWeights_0=1) {
    alphaBoolBlend {blendOption=Linear}
    alphaScaleBiasClamp {scale=0.6}
    boneToModify {boneName=dyn_R_front_ribbon_B_2}
    constraintSetup [{targetBone {boneName=thigh_r}}]
    constraintWeights [1]
   }
   n28 = "Constraint - dyn_R_front_ribbon_B_3 (P)"(ComponentPose=@n27, Alpha=@n25, ConstraintWeights_0=1) {
    alphaBoolBlend {blendOption=Linear}
    alphaScaleBiasClamp {scale=0.6}
    boneToModify {boneName=dyn_R_front_ribbon_B_3}
    constraintSetup [{targetBone {boneName=thigh_r}}]
    constraintWeights [1]
   }
   n29 = "Constraint - dyn_L_front_ribbon_base (P)"(ComponentPose=@n28, Alpha=1, ConstraintWeights_0=1) {
    alphaBoolBlend {blendOption=Linear}
    boneToModify {boneName=dyn_L_front_ribbon_base}
    constraintSetup [{targetBone {boneName=thigh_l}}]
    constraintWeights [1]
   }
   n30 = "Constraint - dyn_L_front_ribbon_1 (P)"(ComponentPose=@n29, Alpha=1, ConstraintWeights_0=1) {
    alphaBoolBlend {blendOption=Linear}
    boneToModify {boneName=dyn_L_front_ribbon_1}
    constraintSetup [{targetBone {boneName=thigh_l}}]
    constraintWeights [1]
   }
   n31 = Anim Dynamics - Bone: dyn_L_front_ribbon_base!(SimulationSpace:AnimPhysSimSpaceType Enum, ComponentPose=@n30, Alpha=1) -> (Pose) {
    angularDampingOverride=0.9999
    angularSpringConstant=200
    bAngularSpring=true
    bOverrideAngularDamping=true
    bUseGravityOverride=true
    gravityScale=0.35
    simulationSpace=RootRelative
    alphaBoolBlend {blendOption=Linear}
    boundBone {boneName=dyn_L_front_ribbon_base}
    componentLinearAccScale=<1 1 1>
    componentLinearVelScale=<1 1 1>
    gravityOverride {z=-2000}
    physicsBodyDefinitions [{
      boundBone {boneName=dyn_L_front_ribbon_base}
      boxExtents=<15 15 15>
      constraintSetup {angularLimitsMax=<0 10 1.5>, angularLimitsMin=<0 -10 0>, angularTarget=<1 0 0>}
      localJointOffset=<2 0 0>
     }]
   }
   n32 = Anim Dynamics - Bone: dyn_L_front_ribbon_1!(SimulationSpace:AnimPhysSimSpaceType Enum, ComponentPose=@n31, Alpha=1) -> (Pose) {
    angularDampingOverride=0.9999
    angularSpringConstant=200
    bAngularSpring=true
    bOverrideAngularDamping=true
    bUseGravityOverride=true
    gravityScale=0.35
    simulationSpace=RootRelative
    alphaBoolBlend {blendOption=Linear}
    boundBone {boneName=dyn_L_front_ribbon_1}
    componentLinearAccScale=<1 1 1>
    componentLinearVelScale=<1 1 1>
    gravityOverride {z=-2000}
    physicsBodyDefinitions [{
      boundBone {boneName=dyn_L_front_ribbon_1}
      boxExtents=<15 15 15>
      constraintSetup {angularLimitsMax=<0 10 1.5>, angularLimitsMin=<0 -10 -50>, angularTarget=<1 0 0>}
      localJointOffset=<2 0 0>
     }]
   }
   n33 = Anim Dynamics - Bone: dyn_L_front_ribbon_2!(SimulationSpace:AnimPhysSimSpaceType Enum, ComponentPose=@n32, Alpha=1) -> (Pose) {
    angularDampingOverride=0.9999
    angularSpringConstant=200
    bAngularSpring=true
    bOverrideAngularDamping=true
    bUseGravityOverride=true
    gravityScale=0.35
    simulationSpace=RootRelative
    alphaBoolBlend {blendOption=Linear}
    boundBone {boneName=dyn_L_front_ribbon_2}
    componentLinearAccScale=<1 1 1>
    componentLinearVelScale=<1 1 1>
    gravityOverride {z=-2000}
    physicsBodyDefinitions [{
      boundBone {boneName=dyn_L_front_ribbon_2}
      boxExtents=<15 15 15>
      constraintSetup {angularLimitsMax=<0 10 1.5>, angularLimitsMin=<0 -10 -50>, angularTarget=<1 0 0>}
      localJointOffset=<2 0 0>
     }]
   }
   n34 = Anim Dynamics - Bone: dyn_L_front_ribbon_3!(SimulationSpace:AnimPhysSimSpaceType Enum, ComponentPose=@n33, Alpha=1) -> (Pose) {
    angularDampingOverride=0.9999
    angularSpringConstant=200
    bAngularSpring=true
    bOverrideAngularDamping=true
    bUseGravityOverride=true
    gravityScale=0.35
    simulationSpace=RootRelative
    alphaBoolBlend {blendOption=Linear}
    boundBone {boneName=dyn_L_front_ribbon_3}
    componentLinearAccScale=<1 1 1>
    componentLinearVelScale=<1 1 1>
    gravityOverride {z=-2000}
    physicsBodyDefinitions [{
      boundBone {boneName=dyn_L_front_ribbon_3}
      boxExtents=<15 15 15>
      constraintSetup {angularLimitsMax=<0 10 1.5>, angularLimitsMin=<0 -10 -50>, angularTarget=<1 0 0>}
      localJointOffset=<2 0 0>
     }]
   }
   n35 = "Constraint - dyn_L_front_ribbon_base (PP)"!(ComponentPose=@n34, Alpha=1, ConstraintWeights_0=0.7, ConstraintWeights_1:float=0.3) -> (Pose) {
    alphaBoolBlend {blendOption=Linear}
    boneToModify {boneName=dyn_L_front_ribbon_base}
    constraintSetup [{targetBone {boneName=thigh_l}}, {targetBone {boneName=pelvis}}]
    constraintWeights [1,1]
   }
   n36 = "Constraint - dyn_L_front_ribbon_1 (P)"(ComponentPose=@n35, Alpha=1, ConstraintWeights_0=1) {
    alphaBoolBlend {blendOption=Linear}
    boneToModify {boneName=dyn_L_front_ribbon_1}
    constraintSetup [{targetBone {boneName=thigh_l}}]
    constraintWeights [1]
   }
   n37 = "Constraint - dyn_L_front_ribbon_2 (P)"(ComponentPose=@n36, Alpha=1, ConstraintWeights_0=1) {
    alphaBoolBlend {blendOption=Linear}
    boneToModify {boneName=dyn_L_front_ribbon_2}
    constraintSetup [{targetBone {boneName=thigh_l}}]
    constraintWeights [1]
   }
   n38 = "Constraint - dyn_L_front_ribbon_3 (P)"(ComponentPose=@n37, Alpha=1, ConstraintWeights_0=1) {
    alphaBoolBlend {blendOption=Linear}
    boneToModify {boneName=dyn_L_front_ribbon_3}
    constraintSetup [{targetBone {boneName=thigh_l}}]
    constraintWeights [1]
   }
   n39 = Anim Dynamics - Bone: dyn_L_front_ribbon_base(ComponentPose=@n38, Alpha=1) {
    angularDampingOverride=0.9999
    angularSpringConstant=200
    bAngularSpring=true
    bOverrideAngularDamping=true
    bUseGravityOverride=true
    gravityScale=0.35
    simulationSpace=RootRelative
    alphaBoolBlend {blendOption=Linear}
    boundBone {boneName=dyn_L_front_ribbon_base}
    componentLinearAccScale=<1 1 1>
    componentLinearVelScale=<1 1 1>
    gravityOverride {z=-2000}
    physicsBodyDefinitions [{
      boundBone {boneName=dyn_L_front_ribbon_base}
      boxExtents=<15 15 15>
      constraintSetup {angularLimitsMax=<0 5 -2>, angularLimitsMin=<0 3 -30>, angularTarget=<1 0 0>}
      localJointOffset=<2 0 0>
     }]
   }
   n40 = Anim Dynamics - Bone: dyn_L_front_ribbon_1(ComponentPose=@n39, Alpha=1) {
    angularDampingOverride=0.9999
    angularSpringConstant=200
    bAngularSpring=true
    bOverrideAngularDamping=true
    bUseGravityOverride=true
    gravityScale=0.35
    simulationSpace=RootRelative
    alphaBoolBlend {blendOption=Linear}
    boundBone {boneName=dyn_L_front_ribbon_1}
    componentLinearAccScale=<1 1 1>
    componentLinearVelScale=<1 1 1>
    gravityOverride {z=-2000}
    physicsBodyDefinitions [{
      boundBone {boneName=dyn_L_front_ribbon_1}
      boxExtents=<15 15 15>
      constraintSetup {angularLimitsMax=<0 5 -2>, angularLimitsMin=<0 3 -30>, angularTarget=<1 0 0>}
      localJointOffset=<2 0 0>
     }]
   }
   n41 = Anim Dynamics - Bone: dyn_L_front_ribbon_2(ComponentPose=@n40, Alpha=1) {
    angularDampingOverride=0.9999
    angularSpringConstant=200
    bAngularSpring=true
    bOverrideAngularDamping=true
    bUseGravityOverride=true
    gravityScale=0.35
    simulationSpace=RootRelative
    alphaBoolBlend {blendOption=Linear}
    boundBone {boneName=dyn_L_front_ribbon_2}
    componentLinearAccScale=<1 1 1>
    componentLinearVelScale=<1 1 1>
    gravityOverride {z=-2000}
    physicsBodyDefinitions [{
      boundBone {boneName=dyn_L_front_ribbon_2}
      boxExtents=<15 15 15>
      constraintSetup {angularLimitsMax=<0 5 -2>, angularLimitsMin=<0 3 -30>, angularTarget=<1 0 0>}
      localJointOffset=<2 0 0>
     }]
   }
   n42 = Anim Dynamics - Bone: dyn_L_front_ribbon_3(ComponentPose=@n41, Alpha=1) {
    angularDampingOverride=0.9999
    angularSpringConstant=200
    bAngularSpring=true
    bOverrideAngularDamping=true
    bUseGravityOverride=true
    gravityScale=0.35
    simulationSpace=RootRelative
    alphaBoolBlend {blendOption=Linear}
    boundBone {boneName=dyn_L_front_ribbon_3}
    componentLinearAccScale=<1 1 1>
    componentLinearVelScale=<1 1 1>
    gravityOverride {z=-2000}
    physicsBodyDefinitions [{
      boundBone {boneName=dyn_L_front_ribbon_3}
      boxExtents=<15 15 15>
      constraintSetup {angularLimitsMax=<0 5 -2>, angularLimitsMin=<0 3 -30>, angularTarget=<1 0 0>}
      localJointOffset=<2 0 0>
     }]
   }
   n43 = "Transform (Modify) Bone - Bone: dyn_L_front_ribbon_base"(ComponentPose=@n42, Alpha=1, Translation="0.5,1,0.4", Rotation="-30,10,0", Scale="1,1,1") {
    rotationMode=BMM_Additive
    rotationSpace=BCS_BoneSpace
    translationMode=BMM_Additive
    translationSpace=BCS_BoneSpace
    alphaBoolBlend {blendOption=Linear}
    alphaScaleBiasClamp {bInterpResult=true, interpSpeedDecreasing=20, interpSpeedIncreasing=20}
    boneToModify {boneName=dyn_L_front_ribbon_base}
    rotation {pitch=-30, yaw=10}
    translation=<0.5 1 0.4>
   }
   n44 = "Transform (Modify) Bone - Bone: dyn_L_front_ribbon_1"(ComponentPose=@n43, Alpha=1, Translation="0,0,0", Rotation="10,0,0", Scale="1,1,1") {
    rotationMode=BMM_Additive
    rotationSpace=BCS_BoneSpace
    alphaBoolBlend {blendOption=Linear}
    alphaScaleBiasClamp {bInterpResult=true, interpSpeedDecreasing=20, interpSpeedIncreasing=20}
    boneToModify {boneName=dyn_L_front_ribbon_1}
    rotation {pitch=10}
   }
   n45 = "Transform (Modify) Bone - Bone: dyn_L_front_ribbon_2"(ComponentPose=@n44, Alpha=1, Translation="0,0,0", Rotation="0,30,0", Scale="1,1,1") {
    rotationMode=BMM_Additive
    rotationSpace=BCS_BoneSpace
    alphaBoolBlend {blendOption=Linear}
    alphaScaleBiasClamp {bInterpResult=true, interpSpeedDecreasing=20, interpSpeedIncreasing=20}
    boneToModify {boneName=dyn_L_front_ribbon_2}
    rotation {yaw=30}
   }
   n46 = "Transform (Modify) Bone - Bone: dyn_L_front_ribbon_3"(ComponentPose=@n45, Alpha=1, Translation="0,0,0", Rotation="0,30,0", Scale="1,1,1") {
    rotationMode=BMM_Additive
    rotationSpace=BCS_BoneSpace
    alphaBoolBlend {blendOption=Linear}
    alphaScaleBiasClamp {bInterpResult=true, interpSpeedDecreasing=20, interpSpeedIncreasing=20}
    boneToModify {boneName=dyn_L_front_ribbon_3}
    rotation {yaw=30}
   }
   n47 = Anim Dynamics - Bone: dyn_L_back_ribbon_base!(SimulationSpace:AnimPhysSimSpaceType Enum, ComponentPose=@n46, Alpha=1) -> (Pose) {
    angularDampingOverride=1
    angularSpringConstant=400
    bAngularSpring=true
    bOverrideAngularDamping=true
    bUseGravityOverride=true
    gravityScale=0.35
    simulationSpace=RootRelative
    alphaBoolBlend {blendOption=Linear}
    boundBone {boneName=dyn_L_back_ribbon_base}
    componentLinearAccScale=<1 1 1>
    componentLinearVelScale=<1 1 1>
    gravityOverride {z=-2200}
    physicsBodyDefinitions [{
      boundBone {boneName=dyn_L_back_ribbon_base}
      boxExtents=<12 12 12>
      constraintSetup {angularLimitsMax=<0 0 3>, angularLimitsMin=<0 0 0>, angularTarget=<1 0 0>}
      localJointOffset=<2 0 0>
     }]
   }
   n48 = Anim Dynamics - Bone: dyn_L_back_ribbon_A_1!(SimulationSpace:AnimPhysSimSpaceType Enum, ComponentPose=@n47, Alpha=1) -> (Pose) {
    angularDampingOverride=1
    angularSpringConstant=400
    bAngularSpring=true
    bOverrideAngularDamping=true
    bUseGravityOverride=true
    gravityScale=0.35
    simulationSpace=RootRelative
    alphaBoolBlend {blendOption=Linear}
    boundBone {boneName=dyn_L_back_ribbon_A_1}
    componentLinearAccScale=<1 1 1>
    componentLinearVelScale=<1 1 1>
    gravityOverride {z=-2200}
    physicsBodyDefinitions [{
      boundBone {boneName=dyn_L_back_ribbon_A_1}
      boxExtents=<12 12 12>
      constraintSetup {angularLimitsMax=<0 5 0>, angularLimitsMin=<0 -40 -50>, angularTarget=<1 0 0>}
      localJointOffset=<2 0 0>
     }]
   }
   n49 = Anim Dynamics - Bone: dyn_L_back_ribbon_A_2!(SimulationSpace:AnimPhysSimSpaceType Enum, ComponentPose=@n48, Alpha=1) -> (Pose) {
    angularDampingOverride=1
    angularSpringConstant=400
    bAngularSpring=true
    bOverrideAngularDamping=true
    bUseGravityOverride=true
    gravityScale=0.35
    simulationSpace=RootRelative
    alphaBoolBlend {blendOption=Linear}
    boundBone {boneName=dyn_L_back_ribbon_A_2}
    componentLinearAccScale=<1 1 1>
    componentLinearVelScale=<1 1 1>
    gravityOverride {z=-2200}
    physicsBodyDefinitions [{
      boundBone {boneName=dyn_L_back_ribbon_A_2}
      boxExtents=<12 12 12>
      constraintSetup {angularLimitsMax=<0 5 0>, angularLimitsMin=<0 -40 -50>, angularTarget=<1 0 0>}
      localJointOffset=<2 0 0>
     }]
   }
   n50 = Anim Dynamics - Bone: dyn_L_back_ribbon_B_1!(SimulationSpace:AnimPhysSimSpaceType Enum, ComponentPose=@n49, Alpha=1) -> (Pose) {
    angularDampingOverride=1
    angularSpringConstant=400
    bAngularSpring=true
    bOverrideAngularDamping=true
    bUseGravityOverride=true
    gravityScale=0.35
    simulationSpace=RootRelative
    alphaBoolBlend {blendOption=Linear}
    boundBone {boneName=dyn_L_back_ribbon_B_1}
    componentLinearAccScale=<1 1 1.1>
    componentLinearVelScale=<1 1 1.1>
    gravityOverride {z=-2200}
    physicsBodyDefinitions [{
      boundBone {boneName=dyn_L_back_ribbon_B_1}
      boxExtents=<14 14 14>
      constraintSetup {angularLimitsMax=<10 40 5>, angularLimitsMin=<-10 -40 -60>, angularTarget=<1 0 0>}
      localJointOffset=<2 0 0>
     }]
   }
   n51 = Anim Dynamics - Bone: dyn_L_back_ribbon_B_2!(SimulationSpace:AnimPhysSimSpaceType Enum, ComponentPose=@n50, Alpha=1) -> (Pose) {
    angularDampingOverride=1
    angularSpringConstant=400
    bAngularSpring=true
    bOverrideAngularDamping=true
    bUseGravityOverride=true
    gravityScale=0.35
    simulationSpace=RootRelative
    alphaBoolBlend {blendOption=Linear}
    boundBone {boneName=dyn_L_back_ribbon_B_2}
    componentLinearAccScale=<1 1 1.1>
    componentLinearVelScale=<1 1 1.1>
    gravityOverride {z=-2200}
    physicsBodyDefinitions [{
      boundBone {boneName=dyn_L_back_ribbon_B_2}
      boxExtents=<14 14 14>
      constraintSetup {angularLimitsMax=<10 40 10>, angularLimitsMin=<-10 -40 -60>, angularTarget=<1 0 0>}
      localJointOffset=<2 0 0>
     }]
   }
   n52 = Anim Dynamics - Bone: dyn_L_back_ribbon_B_3!(SimulationSpace:AnimPhysSimSpaceType Enum, ComponentPose=@n51, Alpha=1) -> (Pose) {
    angularDampingOverride=1
    angularSpringConstant=400
    bAngularSpring=true
    bOverrideAngularDamping=true
    bUseGravityOverride=true
    gravityScale=0.35
    simulationSpace=RootRelative
    alphaBoolBlend {blendOption=Linear}
    boundBone {boneName=dyn_L_back_ribbon_B_3}
    componentLinearAccScale=<1 1 1.1>
    componentLinearVelScale=<1 1 1.1>
    gravityOverride {z=-2200}
    physicsBodyDefinitions [{
      boundBone {boneName=dyn_L_back_ribbon_B_3}
      boxExtents=<14 14 14>
      constraintSetup {angularLimitsMax=<10 40 20>, angularLimitsMin=<-10 -40 -60>, angularTarget=<1 0 0>}
      localJointOffset=<2 0 0>
     }]
   }
   n53 = "Transform (Modify) Bone - Bone: dyn_L_back_ribbon_B_1"(ComponentPose=@n52, Alpha=1, Translation="0,0,0", Rotation="0,-15,0", Scale="1,1,1") {
    rotationMode=BMM_Additive
    rotationSpace=BCS_BoneSpace
    alphaBoolBlend {blendOption=Linear}
    alphaScaleBiasClamp {bInterpResult=true, interpSpeedDecreasing=30, interpSpeedIncreasing=30}
    boneToModify {boneName=dyn_L_back_ribbon_B_1}
    rotation {yaw=-15}
   }
   n54 = "Transform (Modify) Bone - Bone: dyn_L_back_ribbon_B_1"!(ComponentPose=@n53, bAlphaBoolEnabled:bool=True, Translation="0,0,0", Rotation="0,-20,0", Scale="1,1,1") -> (Pose) {
    alphaInputType=Bool
    rotationMode=BMM_Additive
    rotationSpace=BCS_BoneSpace
    alphaBoolBlend {blendInTime=0.2, blendOption=Linear, blendOutTime=0.2}
    alphaScaleBiasClamp {bInterpResult=true, interpSpeedDecreasing=30, interpSpeedIncreasing=30}
    boneToModify {boneName=dyn_L_back_ribbon_B_1}
    rotation {yaw=-20}
   }
   n55 = "Transform (Modify) Bone - Bone: dyn_L_back_ribbon_B_2"!(ComponentPose=@n54, bAlphaBoolEnabled:bool=True, Translation="0,0,0", Rotation="0,-20,0", Scale="1,1,1") -> (Pose) {
    alphaInputType=Bool
    rotationMode=BMM_Additive
    rotationSpace=BCS_BoneSpace
    alphaBoolBlend {blendInTime=0.2, blendOption=Linear, blendOutTime=0.2}
    alphaScaleBiasClamp {bInterpResult=true, interpSpeedDecreasing=30, interpSpeedIncreasing=30}
    boneToModify {boneName=dyn_L_back_ribbon_B_2}
    rotation {yaw=-20}
   }
   n56 = "Transform (Modify) Bone - Bone: dyn_L_back_ribbon_A_1"!(ComponentPose=@n55, bAlphaBoolEnabled:bool=True, Translation="0,0,0", Rotation="0,-15,0", Scale="1,1,1") -> (Pose) {
    alphaInputType=Bool
    rotationMode=BMM_Additive
    rotationSpace=BCS_BoneSpace
    alphaBoolBlend {blendInTime=0.2, blendOption=Linear, blendOutTime=0.2}
    alphaScaleBiasClamp {bInterpResult=true, interpSpeedDecreasing=30, interpSpeedIncreasing=30}
    boneToModify {boneName=dyn_L_back_ribbon_A_1}
    rotation {yaw=-15}
   }
   n57 = "Transform (Modify) Bone - Bone: dyn_L_back_ribbon_A_2"!(ComponentPose=@n56, bAlphaBoolEnabled:bool=True, Translation="0,0,0", Rotation="0,-15,0", Scale="1,1,1") -> (Pose) {
    alphaInputType=Bool
    rotationMode=BMM_Additive
    rotationSpace=BCS_BoneSpace
    alphaBoolBlend {blendInTime=0.2, blendOption=Linear, blendOutTime=0.2}
    alphaScaleBiasClamp {bInterpResult=true, interpSpeedDecreasing=30, interpSpeedIncreasing=30}
    boneToModify {boneName=dyn_L_back_ribbon_A_2}
    rotation {yaw=-15}
   }
   n58 = Anim Dynamics - Bone: dyn_L_ankleFrill_outer!(SimulationSpace:AnimPhysSimSpaceType Enum, ComponentPose=@n57, Alpha=1) -> (Pose) {
    angularDampingOverride=1
    angularSpringConstant=1000
    bAngularSpring=true
    bOverrideAngularDamping=true
    bUseGravityOverride=true
    gravityScale=0.35
    simulationSpace=RootRelative
    alphaBoolBlend {blendOption=Linear}
    boundBone {boneName=dyn_L_ankleFrill_outer}
    componentLinearAccScale=<1 1 1>
    componentLinearVelScale=<1 1 1>
    gravityOverride {z=-100}
    physicsBodyDefinitions [{
      boundBone {boneName=dyn_L_ankleFrill_outer}
      boxExtents=<7 7 7>
      constraintSetup {angularLimitsMax=<0 0 20>, angularLimitsMin=<0 -30 0>, angularTarget=<1 0 0>}
      localJointOffset=<-2 0 0>
     }]
   }
   n59 = Anim Dynamics - Bone: dyn_R_ankleFrill_outer!(SimulationSpace:AnimPhysSimSpaceType Enum, ComponentPose=@n58, Alpha=1) -> (Pose) {
    angularDampingOverride=1
    angularSpringConstant=1000
    bAngularSpring=true
    bOverrideAngularDamping=true
    bUseGravityOverride=true
    gravityScale=0.35
    simulationSpace=RootRelative
    alphaBoolBlend {blendOption=Linear}
    boundBone {boneName=dyn_R_ankleFrill_outer}
    componentLinearAccScale=<1 1 1>
    componentLinearVelScale=<1 1 1>
    gravityOverride {z=-100}
    physicsBodyDefinitions [{
      boundBone {boneName=dyn_R_ankleFrill_outer}
      boxExtents=<7 7 7>
      constraintSetup {angularLimitsMax=<0 0 20>, angularLimitsMin=<0 -30 0>, angularTarget=<1 0 0>}
      localJointOffset=<2 0 0>
     }]
   }
   n60 = Anim Dynamics - Bone: dyn_L_ankleFrill_inner!(SimulationSpace:AnimPhysSimSpaceType Enum, ComponentPose=@n59, Alpha=1) -> (Pose) {
    angularDampingOverride=1
    angularSpringConstant=1000
    bAngularSpring=true
    bOverrideAngularDamping=true
    bUseGravityOverride=true
    gravityScale=0.35
    simulationSpace=RootRelative
    alphaBoolBlend {blendOption=Linear}
    boundBone {boneName=dyn_L_ankleFrill_inner}
    componentLinearAccScale=<1 1 1>
    componentLinearVelScale=<1 1 1>
    gravityOverride {z=-100}
    physicsBodyDefinitions [{
      boundBone {boneName=dyn_L_ankleFrill_inner}
      boxExtents=<7 7 7>
      constraintSetup {angularLimitsMax=<0 30 20>, angularLimitsMin=<0 0 0>, angularTarget=<1 0 0>}
      localJointOffset=<-2 0 0>
     }]
   }
   n61 = Anim Dynamics - Bone: dyn_R_ankleFrill_inner!(SimulationSpace:AnimPhysSimSpaceType Enum, ComponentPose=@n60, Alpha=1) -> (Pose) {
    angularDampingOverride=1
    angularSpringConstant=1000
    bAngularSpring=true
    bOverrideAngularDamping=true
    bUseGravityOverride=true
    gravityScale=0.35
    simulationSpace=RootRelative
    alphaBoolBlend {blendOption=Linear}
    boundBone {boneName=dyn_R_ankleFrill_inner}
    componentLinearAccScale=<1 1 1>
    componentLinearVelScale=<1 1 1>
    gravityOverride {z=-100}
    physicsBodyDefinitions [{
      boundBone {boneName=dyn_R_ankleFrill_inner}
      boxExtents=<7 7 7>
      constraintSetup {angularLimitsMax=<0 30 20>, angularLimitsMin=<0 0 0>, angularTarget=<1 0 0>}
      localJointOffset=<2 0 0>
     }]
   }
   n62 = Anim Dynamics - Bone: dyn_sphere!(SimulationSpace:AnimPhysSimSpaceType Enum, ComponentPose=@n61, Alpha=1) -> (Pose) {
    angularDampingOverride=1
    angularSpringConstant=10000
    bAngularSpring=true
    bOverrideAngularDamping=true
    bUseGravityOverride=true
    gravityScale=0.35
    simulationSpace=RootRelative
    alphaBoolBlend {blendOption=Linear}
    boundBone {boneName=dyn_sphere}
    componentLinearAccScale=<1 1 1>
    componentLinearVelScale=<1 1 1>
    gravityOverride {z=-980}
    physicsBodyDefinitions [{
      boundBone {boneName=dyn_sphere}
      boxExtents=<10 10 10>
      constraintSetup {angularLimitsMax=<0 40 40>, angularLimitsMin=<0 -40 -40>, angularTarget=<1 0 0>}
      localJointOffset=<4 0 0>
     }]
   }
   n63 = "Constraint - dyn_R_top_hipRuffle (PP)"!(ComponentPose=@n62, Alpha=1, ConstraintWeights_0=0.2, ConstraintWeights_1:float=0.8) -> (Pose) {
    alphaBoolBlend {blendOption=Linear}
    boneToModify {boneName=dyn_R_top_hipRuffle}
    constraintSetup [{targetBone {boneName=thigh_r}}, {targetBone {boneName=pelvis}}]
    constraintWeights [1,1]
   }
   n64 = "Constraint - dyn_R_mid_hipRuffles (PP)"!(ComponentPose=@n63, Alpha=1, ConstraintWeights_0=0.65, ConstraintWeights_1:float=0.35) -> (Pose) {
    alphaBoolBlend {blendOption=Linear}
    boneToModify {boneName=dyn_R_mid_hipRuffles}
    constraintSetup [{targetBone {boneName=thigh_r}}, {targetBone {boneName=pelvis}}]
    constraintWeights [1,1]
   }
   n65 = "Constraint - dyn_R_bottom_hipRuffles (PP)"!(ComponentPose=@n64, Alpha=1, ConstraintWeights_0=0.8, ConstraintWeights_1:float=0.2) -> (Pose) {
    alphaBoolBlend {blendOption=Linear}
    boneToModify {boneName=dyn_R_bottom_hipRuffles}
    constraintSetup [{targetBone {boneName=thigh_r}}, {targetBone {boneName=pelvis}}]
    constraintWeights [1,1]
   }
   n66 = Anim Dynamics - Bone: dyn_R_top_hipRuffle!(SimulationSpace:AnimPhysSimSpaceType Enum, ComponentPose=@n65, Alpha=1) -> (Pose) {
    angularDampingOverride=1
    angularSpringConstant=10
    bAngularSpring=true
    bLinearSpring=true
    bOverrideAngularDamping=true
    bUseGravityOverride=true
    gravityScale=0.35
    linearSpringConstant=100
    simulationSpace=RootRelative
    alphaBoolBlend {blendOption=Linear}
    boundBone {boneName=dyn_R_top_hipRuffle}
    componentLinearAccScale=<0.1 0.1 0.31>
    componentLinearVelScale=<0.1 0.1 0.31>
    gravityOverride {z=-700}
    physicsBodyDefinitions [{
      boundBone {boneName=dyn_R_top_hipRuffle}
      boxExtents=<12 12 12>
      constraintSetup {angularLimitsMax=<0 0 40>, angularLimitsMin=<0 0 3>, angularTarget=<1 0 0>}
      localJointOffset=<-2.5 0 0>
     }]
    relativeSpaceBone {boneName=pelvis}
   }
   n67 = Anim Dynamics - Bone: dyn_R_mid_hipRuffles!(SimulationSpace:AnimPhysSimSpaceType Enum, ComponentPose=@n66, Alpha=1) -> (Pose) {
    angularDampingOverride=1
    angularSpringConstant=10
    bAngularSpring=true
    bLinearSpring=true
    bOverrideAngularDamping=true
    bUseGravityOverride=true
    gravityScale=0.35
    linearSpringConstant=100
    simulationSpace=RootRelative
    alphaBoolBlend {blendOption=Linear}
    boundBone {boneName=dyn_R_mid_hipRuffles}
    componentLinearAccScale=<0.1 0.1 0.31>
    componentLinearVelScale=<0.1 0.1 0.31>
    gravityOverride {z=-700}
    physicsBodyDefinitions [{
      boundBone {boneName=dyn_R_mid_hipRuffles}
      boxExtents=<12 12 12>
      constraintSetup {angularLimitsMax=<0 0 50>, angularLimitsMin=<0 0 3>, angularTarget=<1 0 0>}
      localJointOffset=<-2.5 0 0>
     }]
    relativeSpaceBone {boneName=pelvis}
   }
   n68 = Anim Dynamics - Bone: dyn_R_bottom_hipRuffles!(SimulationSpace:AnimPhysSimSpaceType Enum, ComponentPose=@n67, Alpha=1) -> (Pose) {
    angularDampingOverride=1
    angularSpringConstant=10
    bAngularSpring=true
    bLinearSpring=true
    bOverrideAngularDamping=true
    bUseGravityOverride=true
    gravityScale=0.35
    linearSpringConstant=100
    simulationSpace=RootRelative
    alphaBoolBlend {blendOption=Linear}
    boundBone {boneName=dyn_R_bottom_hipRuffles}
    componentLinearAccScale=<0.1 0.1 0.31>
    componentLinearVelScale=<0.1 0.1 0.31>
    gravityOverride {z=-700}
    physicsBodyDefinitions [{
      boundBone {boneName=dyn_R_bottom_hipRuffles}
      boxExtents=<12 12 12>
      constraintSetup {angularLimitsMax=<0 0 50>, angularLimitsMin=<0 0 3>, angularTarget=<1 0 0>}
      localJointOffset=<-2.5 0 0>
     }]
    relativeSpaceBone {boneName=pelvis}
   }
   n69 = "Constraint - dyn_L_top_hipRuffle (PP)"!(ComponentPose=@n68, Alpha=1, ConstraintWeights_0=0.2, ConstraintWeights_1:float=0.8) -> (Pose) {
    alphaBoolBlend {blendOption=Linear}
    boneToModify {boneName=dyn_L_top_hipRuffle}
    constraintSetup [{targetBone {boneName=thigh_l}}, {targetBone {boneName=pelvis}}]
    constraintWeights [1,1]
   }
   n70 = "Constraint - dyn_L_mid_hipRuffles (PP)"!(ComponentPose=@n69, Alpha=1, ConstraintWeights_0=0.65, ConstraintWeights_1:float=0.35) -> (Pose) {
    alphaBoolBlend {blendOption=Linear}
    boneToModify {boneName=dyn_L_mid_hipRuffles}
    constraintSetup [{targetBone {boneName=thigh_l}}, {targetBone {boneName=pelvis}}]
    constraintWeights [1,1]
   }
   n71 = "Constraint - dyn_L_bottom_hipRuffles (PP)"!(ComponentPose=@n70, Alpha=1, ConstraintWeights_0=0.8, ConstraintWeights_1:float=0.2) -> (Pose) {
    alphaBoolBlend {blendOption=Linear}
    boneToModify {boneName=dyn_L_bottom_hipRuffles}
    constraintSetup [{targetBone {boneName=thigh_l}}, {targetBone {boneName=pelvis}}]
    constraintWeights [1,1]
   }
   n72 = Anim Dynamics - Bone: dyn_L_top_hipRuffle!(SimulationSpace:AnimPhysSimSpaceType Enum, ComponentPose=@n71, Alpha=1) -> (Pose) {
    angularDampingOverride=1
    angularSpringConstant=10
    bAngularSpring=true
    bLinearSpring=true
    bOverrideAngularDamping=true
    bUseGravityOverride=true
    gravityScale=0.35
    linearSpringConstant=100
    simulationSpace=RootRelative
    alphaBoolBlend {blendOption=Linear}
    boundBone {boneName=dyn_L_top_hipRuffle}
    componentLinearAccScale=<0.1 0.1 0.31>
    componentLinearVelScale=<0.1 0.1 0.31>
    gravityOverride {z=-750}
    physicsBodyDefinitions [{
      boundBone {boneName=dyn_L_top_hipRuffle}
      boxExtents=<12 12 12>
      constraintSetup {angularLimitsMax=<0 0 40>, angularLimitsMin=<0 0 3>, angularTarget=<1 0 0>}
      localJointOffset=<2.5 0 0>
     }]
    relativeSpaceBone {boneName=pelvis}
   }
   n73 = Anim Dynamics - Bone: dyn_L_mid_hipRuffles!(SimulationSpace:AnimPhysSimSpaceType Enum, ComponentPose=@n72, Alpha=1) -> (Pose) {
    angularDampingOverride=1
    angularSpringConstant=10
    bAngularSpring=true
    bLinearSpring=true
    bOverrideAngularDamping=true
    bUseGravityOverride=true
    gravityScale=0.35
    linearSpringConstant=100
    simulationSpace=RootRelative
    alphaBoolBlend {blendOption=Linear}
    boundBone {boneName=dyn_L_mid_hipRuffles}
    componentLinearAccScale=<0.1 0.1 0.31>
    componentLinearVelScale=<0.1 0.1 0.31>
    gravityOverride {z=-750}
    physicsBodyDefinitions [{
      boundBone {boneName=dyn_L_mid_hipRuffles}
      boxExtents=<12 12 12>
      constraintSetup {angularLimitsMax=<0 0 50>, angularLimitsMin=<0 0 3>, angularTarget=<1 0 0>}
      localJointOffset=<2.5 0 0>
     }]
    relativeSpaceBone {boneName=pelvis}
   }
   n74 = Anim Dynamics - Bone: dyn_L_bottom_hipRuffles!(SimulationSpace:AnimPhysSimSpaceType Enum, ComponentPose=@n73, Alpha=1) -> (Pose) {
    angularDampingOverride=1
    angularSpringConstant=10
    bAngularSpring=true
    bLinearSpring=true
    bOverrideAngularDamping=true
    bUseGravityOverride=true
    gravityScale=0.35
    linearSpringConstant=100
    simulationSpace=RootRelative
    alphaBoolBlend {blendOption=Linear}
    boundBone {boneName=dyn_L_bottom_hipRuffles}
    componentLinearAccScale=<0.1 0.1 0.31>
    componentLinearVelScale=<0.1 0.1 0.31>
    gravityOverride {z=-750}
    physicsBodyDefinitions [{
      boundBone {boneName=dyn_L_bottom_hipRuffles}
      boxExtents=<12 12 12>
      constraintSetup {angularLimitsMax=<0 0 50>, angularLimitsMin=<0 0 3>, angularTarget=<1 0 0>}
      localJointOffset=<2.5 0 0>
     }]
    relativeSpaceBone {boneName=pelvis}
   }
   n75 = "Constraint - dyn_L_ankle_ribbon_inner_1 (RR)"!(ComponentPose=@n74, Alpha=1, ConstraintWeights_0=1, ConstraintWeights_1:float=0) -> (Pose) {
    alphaBoolBlend {blendOption=Linear}
    boneToModify {boneName=dyn_L_ankle_ribbon_inner_1}
    constraintSetup [
     {transformType=Rotation, targetBone {boneName=dyn_L_ankleFrill_inner}}
     {transformType=Rotation, targetBone {boneName=dyn_L_ankleFrill_outer}}
    ]
    constraintWeights [1,1]
   }
   n76 = "Constraint - dyn_L_ankle_ribbon_outer_1 (RR)"!(ComponentPose=@n75, Alpha=1, ConstraintWeights_0=0, ConstraintWeights_1:float=1) -> (Pose) {
    alphaBoolBlend {blendOption=Linear}
    boneToModify {boneName=dyn_L_ankle_ribbon_outer_1}
    constraintSetup [
     {transformType=Rotation, targetBone {boneName=dyn_L_ankleFrill_inner}}
     {transformType=Rotation, targetBone {boneName=dyn_L_ankleFrill_outer}}
    ]
    constraintWeights [1,1]
   }
   n77 = "Constraint - dyn_R_ankle_ribbon_inner_1 (RR)"!(ComponentPose=@n76, Alpha=1, ConstraintWeights_0=1, ConstraintWeights_1:float=0) -> (Pose) {
    alphaBoolBlend {blendOption=Linear}
    boneToModify {boneName=dyn_R_ankle_ribbon_inner_1}
    constraintSetup [
     {transformType=Rotation, targetBone {boneName=dyn_R_ankleFrill_inner}}
     {transformType=Rotation, targetBone {boneName=dyn_R_ankleFrill_outer}}
    ]
    constraintWeights [1,1]
   }
   n78 = "Constraint - dyn_R_ankle_ribbon_outer_1 (RR)"!(ComponentPose=@n77, Alpha=1, ConstraintWeights_0=0, ConstraintWeights_1:float=1) -> (Pose) {
    alphaBoolBlend {blendOption=Linear}
    boneToModify {boneName=dyn_R_ankle_ribbon_outer_1}
    constraintSetup [
     {transformType=Rotation, targetBone {boneName=dyn_R_ankleFrill_inner}}
     {transformType=Rotation, targetBone {boneName=dyn_R_ankleFrill_outer}}
    ]
    constraintWeights [1,1]
   }
   n79 = Anim Dynamics - Bone: dyn_L_ankle_ribbon_inner_1!(SimulationSpace:AnimPhysSimSpaceType Enum, ComponentPose=@n78, Alpha=1) -> (Pose) {
    angularDampingOverride=1
    angularSpringConstant=400
    bAngularSpring=true
    bOverrideAngularDamping=true
    bUseGravityOverride=true
    gravityScale=0.35
    simulationSpace=RootRelative
    alphaBoolBlend {blendOption=Linear}
    boundBone {boneName=dyn_L_ankle_ribbon_inner_1}
    componentLinearAccScale=<1 1 1>
    componentLinearVelScale=<1 1 1>
    gravityOverride {z=-2000}
    physicsBodyDefinitions [{
      boundBone {boneName=dyn_L_ankle_ribbon_inner_1}
      boxExtents=<12 12 12>
      constraintSetup {angularLimitsMax=<0 10 5>, angularLimitsMin=<0 -10 -30>, angularTarget=<1 0 0>}
      localJointOffset=<-2 0 0>
     }]
   }
   n80 = Anim Dynamics - Bone: dyn_L_ankle_ribbon_outer_1!(SimulationSpace:AnimPhysSimSpaceType Enum, ComponentPose=@n79, Alpha=1) -> (Pose) {
    angularDampingOverride=1
    angularSpringConstant=400
    bAngularSpring=true
    bOverrideAngularDamping=true
    bUseGravityOverride=true
    gravityScale=0.35
    simulationSpace=RootRelative
    alphaBoolBlend {blendOption=Linear}
    boundBone {boneName=dyn_L_ankle_ribbon_outer_1}
    componentLinearAccScale=<1 1 1>
    componentLinearVelScale=<1 1 1>
    gravityOverride {z=-2000}
    physicsBodyDefinitions [{
      boundBone {boneName=dyn_L_ankle_ribbon_outer_1}
      boxExtents=<12 12 12>
      constraintSetup {angularLimitsMax=<0 10 5>, angularLimitsMin=<0 -10 -30>, angularTarget=<1 0 0>}
      localJointOffset=<-2 0 0>
     }]
   }
   n81 = Anim Dynamics - Bone: dyn_L_ankle_ribbon_inner_2!(SimulationSpace:AnimPhysSimSpaceType Enum, ComponentPose=@n80, Alpha=1) -> (Pose) {
    angularDampingOverride=1
    angularSpringConstant=400
    bAngularSpring=true
    bOverrideAngularDamping=true
    bUseGravityOverride=true
    gravityScale=0.35
    simulationSpace=RootRelative
    alphaBoolBlend {blendOption=Linear}
    boundBone {boneName=dyn_L_ankle_ribbon_inner_2}
    componentLinearAccScale=<1 1 1>
    componentLinearVelScale=<1 1 1>
    gravityOverride {z=-2000}
    physicsBodyDefinitions [{
      boundBone {boneName=dyn_L_ankle_ribbon_inner_2}
      boxExtents=<12 12 12>
      constraintSetup {angularLimitsMax=<0 10 5>, angularLimitsMin=<0 -10 -30>, angularTarget=<1 0 0>}
      localJointOffset=<-2 0 0>
     }]
   }
   n82 = Anim Dynamics - Bone: dyn_L_ankle_ribbon_outer_2!(SimulationSpace:AnimPhysSimSpaceType Enum, ComponentPose=@n81, Alpha=1) -> (Pose) {
    angularDampingOverride=1
    angularSpringConstant=400
    bAngularSpring=true
    bOverrideAngularDamping=true
    bUseGravityOverride=true
    gravityScale=0.35
    simulationSpace=RootRelative
    alphaBoolBlend {blendOption=Linear}
    boundBone {boneName=dyn_L_ankle_ribbon_outer_2}
    componentLinearAccScale=<1 1 1>
    componentLinearVelScale=<1 1 1>
    gravityOverride {z=-2000}
    physicsBodyDefinitions [{
      boundBone {boneName=dyn_L_ankle_ribbon_outer_2}
      boxExtents=<12 12 12>
      constraintSetup {angularLimitsMax=<0 10 5>, angularLimitsMin=<0 -10 -30>, angularTarget=<1 0 0>}
      localJointOffset=<-2 0 0>
     }]
   }
   n83 = Anim Dynamics - Bone: dyn_R_ankle_ribbon_inner_1!(SimulationSpace:AnimPhysSimSpaceType Enum, ComponentPose=@n82, Alpha=1) -> (Pose) {
    angularDampingOverride=1
    angularSpringConstant=400
    bAngularSpring=true
    bOverrideAngularDamping=true
    bUseGravityOverride=true
    gravityScale=0.35
    simulationSpace=RootRelative
    alphaBoolBlend {blendOption=Linear}
    boundBone {boneName=dyn_R_ankle_ribbon_inner_1}
    componentLinearAccScale=<1 1 1>
    componentLinearVelScale=<1 1 1>
    gravityOverride {z=-2000}
    physicsBodyDefinitions [{
      boundBone {boneName=dyn_R_ankle_ribbon_inner_1}
      boxExtents=<12 12 12>
      constraintSetup {angularLimitsMax=<0 10 5>, angularLimitsMin=<0 -10 -30>, angularTarget=<1 0 0>}
      localJointOffset=<2 0 0>
     }]
   }
   n84 = Anim Dynamics - Bone: dyn_R_ankle_ribbon_outer_1!(SimulationSpace:AnimPhysSimSpaceType Enum, ComponentPose=@n83, Alpha=1) -> (Pose) {
    angularDampingOverride=1
    angularSpringConstant=400
    bAngularSpring=true
    bOverrideAngularDamping=true
    bUseGravityOverride=true
    gravityScale=0.35
    simulationSpace=RootRelative
    alphaBoolBlend {blendOption=Linear}
    boundBone {boneName=dyn_R_ankle_ribbon_outer_1}
    componentLinearAccScale=<1 1 1>
    componentLinearVelScale=<1 1 1>
    gravityOverride {z=-2000}
    physicsBodyDefinitions [{
      boundBone {boneName=dyn_R_ankle_ribbon_outer_1}
      boxExtents=<12 12 12>
      constraintSetup {angularLimitsMax=<0 10 5>, angularLimitsMin=<0 -10 -30>, angularTarget=<1 0 0>}
      localJointOffset=<2 0 0>
     }]
   }
   n85 = Anim Dynamics - Bone: dyn_R_ankle_ribbon_inner_2!(SimulationSpace:AnimPhysSimSpaceType Enum, ComponentPose=@n84, Alpha=1) -> (Pose) {
    angularDampingOverride=1
    angularSpringConstant=400
    bAngularSpring=true
    bOverrideAngularDamping=true
    bUseGravityOverride=true
    gravityScale=0.35
    simulationSpace=RootRelative
    alphaBoolBlend {blendOption=Linear}
    boundBone {boneName=dyn_R_ankle_ribbon_inner_2}
    componentLinearAccScale=<1 1 1>
    componentLinearVelScale=<1 1 1>
    gravityOverride {z=-2000}
    physicsBodyDefinitions [{
      boundBone {boneName=dyn_R_ankle_ribbon_inner_2}
      boxExtents=<12 12 12>
      constraintSetup {angularLimitsMax=<0 10 5>, angularLimitsMin=<0 -10 -30>, angularTarget=<1 0 0>}
      localJointOffset=<2 0 0>
     }]
   }
   n86 = Anim Dynamics - Bone: dyn_R_ankle_ribbon_outer_2!(SimulationSpace:AnimPhysSimSpaceType Enum, ComponentPose=@n85, Alpha=1) -> (Pose) {
    angularDampingOverride=1
    angularSpringConstant=400
    bAngularSpring=true
    bOverrideAngularDamping=true
    bUseGravityOverride=true
    gravityScale=0.35
    simulationSpace=RootRelative
    alphaBoolBlend {blendOption=Linear}
    boundBone {boneName=dyn_R_ankle_ribbon_outer_2}
    componentLinearAccScale=<1 1 1>
    componentLinearVelScale=<1 1 1>
    gravityOverride {z=-2000}
    physicsBodyDefinitions [{
      boundBone {boneName=dyn_R_ankle_ribbon_outer_2}
      boxExtents=<12 12 12>
      constraintSetup {angularLimitsMax=<0 10 5>, angularLimitsMin=<0 -10 -30>, angularTarget=<1 0 0>}
      localJointOffset=<2 0 0>
     }]
   }
   n87 = "Transform (Modify) Bone - Bone: lowerarm_twist_01_l"(ComponentPose=@n86, Alpha=1, Translation="0,0,0", Rotation="0,0,30", Scale="1,1,1") {
    rotationMode=BMM_Additive
    rotationSpace=BCS_BoneSpace
    alphaBoolBlend {blendOption=Linear}
    boneToModify {boneName=lowerarm_twist_01_l}
    rotation {roll=30}
   }
   n88 = "Transform (Modify) Bone - Bone: lowerarm_twist_01_r"(ComponentPose=@n87, Alpha=1, Translation="0,0,0", Rotation="0,0,30", Scale="1,1,1") {
    rotationMode=BMM_Additive
    rotationSpace=BCS_BoneSpace
    alphaBoolBlend {blendOption=Linear}
    boneToModify {boneName=lowerarm_twist_01_r}
    rotation {roll=30}
   }
   Output Pose(Result=Component To Local(ComponentPose=@n88)) {layerGroup=None}
  ]}
 EventGraph {nodes [
   Event BlueprintUpdateAnimation()
   > n1 = Sequence()
   > Set simSpace(simSpace:AnimPhysSimSpaceType Enum) -> (Output_Get:AnimPhysSimSpaceType Enum)
   > Branch(Condition=OR Boolean(A=Get bIsDBNO(), B=Get bIsInFrontEnd()))
   > n2 = Branch(Condition=OR Boolean(A=Get bIsDBNOCarried(), B=Get bIsInFrontEnd()))
   > Set simSpace(simSpace=World)
   Set simSpace(execute=@n2.else)
   n3 = Set isCrouching(execute=@n1.then_1, isCrouching:double="To Float (Boolean)"(InBool=Get bIsCrouching())) -> (Output_Get:double)
   > Set isNotCrouching(isNotCrouching:double=float - float(A=1, B=@n3)) -> (Output_Get:double)
   n4 = BreakTransform(InTransform=GetDeltaTransformFromRefPose(self=GetOwnerSkeletalMeshComponent(), BoneName=thigh_l, BaseName=pelvis))
   > Set L thighBack(L thighBack:double=MapRangeClamped(Value=BreakRotator(InRot=@n4.Rotation).Yaw, InRangeA=-10, InRangeB=0)) -> (Output_Get:double)
  ]}
]
variables [
 isCrouching {type=number}
 isNotCrouching {type=number}
 L thighBack {type=number}
 L_wristRotator {type=Rotator}
 simSpace {type=AnimPhysSimSpaceType, default=World}
]
